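# Additive unlit material that draws a lighthouse lens flare on a camera-facing sprite, pulled by a camera offset in world position offset.
# It layers a streak texture, a radial glow and banded reflex rings built from curve-atlas lookups and rotated by view angle.
# A static switch swaps the color parameters for stage-show colors.
# A screen-space checker pattern adds texture, and depth fade is skipped on mobile and low quality.
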Material /SparksMaps/Environment/S10/LighthouseFlare/M_Sparks_LighthouseFlare
graph {
 properties {
  bAlwaysEvaluateWorldPositionOffset=true
  bAutomaticallySetUsageInEditor=false
  bCastRayTracedShadows=false
  bUseTranslucencyVertexFog=false
  blendMode=BLEND_Additive
  shadingModel=MSM_Unlit
  usedShadingModels=Unlit
 }
 nodes [
  n1 = NamedRerouteUsage()
  Abs(None=@n1)
  n2 = NamedRerouteUsage()
  n3 = NamedRerouteUsage()
  Add(A=NamedRerouteUsage(), B=Multiply(A=NamedRerouteUsage(), B=Add(A=@n2, B=@n3)))
  n4 = ScalarParameter() {defaultValue=0.02, group=Flare Advanced, parameterName=Pattern Scale, sliderMax=0.03, sliderMin=0.01, sortPriority=31}
  n5 = TextureSample(UVs=Multiply(A=ScreenPosition().PixelPosition, B=@n4) {constB=0.02}) {texture=/Game/Maps/Textures/checkerGrid}
  n6 = ScalarParameter() {defaultValue=0.15, group=Flare, parameterName=Pattern Contrast, sliderMax=1, sortPriority=7}
  NamedRerouteDeclaration(None=Add(A=Multiply(A=@n5.RGB, B=@n6) {constB=0.7}, B=OneMinus(None=@n6)) {constA=0.3, constB=0.5}) {name=CheckerBoard Pattern, nodeColor=<rgba 1 0 0.1172 1>}
  n7 = NamedRerouteUsage()
  n8 = NamedRerouteDeclaration(None=Distance(A=TextureCoordinate(), B=Constant2Vector() {g=0.5, r=0.5})) {name=SphereGradient, nodeColor=<rgba 0.5 1 0 1>}
  n9 = Multiply(A=@n8, B=ScalarParameter() {defaultValue=1, group=Flare Advanced, parameterName=Reflex Ch01 Tiling, sortPriority=6}) {constB=2}
  n10 = VertexInterpolator(VS=@n1)
  n11 = Multiply(A=ScalarParameter() {defaultValue=1, group=Flare Advanced, parameterName=Reflex Ch01 View Offset, sortPriority=6}, B=@n10.PS) {constB=-0.25}
  n12 = CurveAtlasRowParameter(CurveTime=Frac(None=Add(A=@n9, B=@n11))) {
   atlas=&/CA_Sparks_LighthouseFlare
   curve=&/C_Sparks_LighthouseFlare_LowFreq
   group=Flare Advanced
   parameterName=Reflex Banding Curve
  }
  n13 = Multiply(A=@n8, B=ScalarParameter() {defaultValue=3, group=Flare Advanced, parameterName=Reflex Ch02 Tiling, sortPriority=6}) {constB=2}
  n14 = Multiply(A=@n10.PS, B=ScalarParameter() {defaultValue=0.75, group=Flare Advanced, parameterName=Reflex Ch02 View Offset, sortPriority=6})
  n15 = CurveAtlasRowParameter(CurveTime=Frac(None=Add(A=@n13, B=@n14))) {
   atlas=&CA_Sparks_LighthouseFlare
   curve=&C_Sparks_LighthouseFlare_LowFreq
   group=Flare Advanced
   parameterName=Reflex Banding Curve
  }
  n16 = Add(A=@n12.R, B=@n15.G)
  n17 = ScalarParameter() {defaultValue=0.6, group=Flare Advanced, parameterName=Reflex Falloff Size, sliderMax=1, sliderMin=0.0001, sortPriority=4}
  n18 = ScalarParameter() {group=Flare Advanced, parameterName=Reflex Size, sliderMax=1, sortPriority=3}
  n19 = ActorPositionWS() {originType=CameraRelative}
  n20 = MaterialFunctionCall!(
   Normalized V1=Constant3Vector() {constant=<rgba 1 0 0 0>}
   Normalized V2=Normalize(VectorInput=TransformPosition(None=@n19) {transformSourceType=TRANSFORMPOSSOURCE_TranslatedWorld, transformType=TRANSFORMPOSSOURCE_Camera})
  ) {materialFunction=/Engine/Functions/Engine_MaterialFunctions02/Math/AngleBetweenVectors}
  n21 = Add(
   A=ScalarParameter() {group=Flare Advanced, parameterName=Reflex Angle, sliderMax=1, sortPriority=5}
   B=NamedRerouteDeclaration(None=OneMinus(None=@n20.0-1)) {name=View Angle, nodeColor=<rgba 0 1 0.8594 1>}
  )
  n22 = Subtract(A=TextureCoordinate()) {constB=0.5}
  n23 = Normalize(VectorInput=VertexInterpolator(VS=MaterialFunctionCall!(Angle=@n21, UV/2DVector=@n22) {materialFunction=/Game/Characters/Player/Male/Medium/Bodies/M_MED_Faux/Materials/MF_2DRotate}.Result).PS)
  n24 = Constant2Vector() {g=1}
  n25 = SmoothStep(Min=Add(A=@n18, B=Multiply(A=@n17) {constB=2}), Max=Subtract(A=@n18, B=@n17), Value=Abs(None=DotProduct(A=@n23, B=@n24))) {constMax=0, constMin=0.7}
  n26 = Multiply(A=Power(Base=SmoothStep(Value=@n7) {constMax=0, constMin=0.65}) {constExponent=3}, B=Multiply(A=Multiply(A=@n16), B=Power(Base=@n25)))
  n27 = StaticSwitchParameter(
   True=MaterialFunctionCall!(CustomUPosition) {materialFunction=/SparksStageShow/Systems/StageManager/FXManager/MaterialFunctions/MF_MojoStageColors}.Primary
   False=VectorParameter() {group=Flare, parameterName=Reflex Color, sortPriority=3, defaultValue=<rgba 1 0.1356 0 1>}.RGB
  ) {parameterName=UseStageShowColor}
  NamedRerouteDeclaration(None=Multiply(
    A=Multiply(A=@n26, B=@n27)
    B=ScalarParameter() {defaultValue=1.2, group=Flare, parameterName=Reflex Multiplier, sliderMax=3, sortPriority=6}
   )) {name=LensReflex, nodeColor=<rgba 1 0 0.0469 1>}
  n28 = TextureSample(UVs=MaterialFunctionCall(
    UVs=TextureCoordinate()
    Texture Scale=ScalarParameter() {defaultValue=0.22, group=Flare Advanced, parameterName=FlareSize, sliderMax=1, sliderMin=0.1, sortPriority=1}
   ) {materialFunction=&ScaleUVsByCenter}.UVs) {texture=&/T_Sparks_Flare_Streaks}
  n29 = ScalarParameter() {defaultValue=3, group=Flare, parameterName=FlareMultiplier, sortPriority=4}
  n30 = Multiply(A=MaterialFunctionCall() {materialFunction=&MF_MojoStageColors}.Primary) {constB=0.5}
  n31 = VectorParameter() {group=Flare, parameterName=Flare Color, sortPriority=1, defaultValue=<rgba 1 0.7197 0.2035 1>}
  NamedRerouteDeclaration(None=Multiply(A=Multiply(A=@n28.B, B=@n29) {constB=2}, B=StaticSwitchParameter(True=@n30, False=@n31.RGB) {parameterName=UseStageShowColor})) {name=LensFlare, nodeColor=<rgba 0 0.3672 1 1>}
  n32 = ScalarParameter() {defaultValue=3000, group=Flare Advanced, parameterName=CameraOffset}
  n33 = WorldPosition()
  n34 = Multiply(
   A=Constant2Vector() {g=1000, r=1000}
   B=ScalarParameter() {defaultValue=8, group=Flare, parameterName=General Size, sliderMax=10, sortPriority=0}
  )
  NamedRerouteDeclaration(None=Add(
    A=MaterialFunctionCall(Offset Amount=@n32, World Position=@n33.XYZ) {materialFunction=/Engine/Functions/Engine_MaterialFunctions02/WorldPositionOffset/CameraOffset}.Camera Offset
    B=MaterialFunctionCall!(Center Location, 0-1 UVs, XY scale=@n34, OverridePivot, "Normals (Optional)") {materialFunction=&Sprite}.World position offset
   )) {name=WPO, nodeColor=<rgba 0.6875 1 0 1>}
  n35 = Power(Base=SmoothStep(
    Min=ScalarParameter() {defaultValue=0.6, group=Flare Advanced, parameterName=Glow Size, sliderMax=1, sortPriority=2}
    Value=NamedRerouteUsage()
   ) {constMax=0, constMin=0.7}) {constExponent=3}
  n36 = StaticSwitchParameter(
   True=MaterialFunctionCall() {materialFunction=&MF_MojoStageColors}.Primary
   False=VectorParameter() {group=Flare, parameterName=GlowColor, sortPriority=2, defaultValue=<rgba 1 0.5729 0.1496 1>}.RGB
  ) {parameterName=UseStageShowColor}
  NamedRerouteDeclaration(None=Multiply(A=Multiply(A=@n35, B=@n36), B=ScalarParameter() {defaultValue=0.7, group=Flare, parameterName=Glow Multiplier, sliderMax=3, sortPriority=5})) {name=Glow, nodeColor=<rgba 0.9922 0 1 1>}
  NamedRerouteUsage()
  n37 = Constant() {r=1}
  ShadingPathSwitch(Default=QualitySwitch(Default=DepthFade() {fadeDistanceDefault=500}, Low=@n37), Mobile=@n37)
 ]
}
